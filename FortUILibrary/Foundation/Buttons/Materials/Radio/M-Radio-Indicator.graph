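# Translucent UI material for a radio button indicator, drawing a check icon that changes between off and on states.
# The ToggleOn scalar blends two branches: a color going from black to a tint plus radial glow, and an alpha-based mask.
# The mask shows a dimmed check when off, and the full check plus soft glow when on.
# The glow comes from one shared exponential radial gradient, and the check texture's UVs are scaled about center.
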Material /FortUILibrary/Foundation/Buttons/Materials/Radio/M-Radio-Indicator
graph {
 properties {
  bUsedWithStaticMesh=false
  blendMode=BLEND_Translucent
  materialDomain=MD_UI
  pixelDepthOffsetMode=PDOM_Legacy
  usedShadingModels=Default Lit
 }
 nodes [
  n1 = MaterialFunctionCall(CenterPosition=Constant2Vector() {g=0.55, r=0.5}, Radius=Constant() {r=0.6}, Density=Constant() {r=1}) {materialFunction=&RadialGradientExponential}
  n2 = ScalarParameter() {parameterName=ToggleOn}
  LinearInterpolate(
   A=Constant3Vector()
   B=Add(A=Constant3Vector() {constant=<rgba 32 32 0 0>}, B=Multiply(A=@n1.RadialGradientExponential) {constB=0.5})
   Alpha=@n2
  )
  n3 = TextureSample(UVs=MaterialFunctionCall(Texture Scale=Constant() {r=0.65}) {materialFunction=&ScaleUVsByCenter}.UVs) {texture=/Game/UI/Foundation/Textures/Icons/Checks/T-Icon-Check-64}
  LinearInterpolate(A=Multiply(A=@n3.A) {constB=0.35}, B=Add(A=@n3.A, B=Multiply(A=@n1.RadialGradientExponential) {constB=0.35}), Alpha=@n2)
 ]
}
